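; Non-participant boss player pawn for the WildEstate ClimbSpill event, with a toggleable balloon prop.
; On BeginPlay it overrides the current weapon's animation proxy table with the ClimbSpill pistol set, spawns a portal entry/exit Niagara effect where cosmetic events can run, and force-hides its health bar.
; SetPropVisibiilty writes the PropVisibility variable, whose OnRep function applies visibility to the Balloon and its children.
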(event Custom|SpawnRiftEffect
  (Networking|CanExecuteCosmeticEvents
    (:True
      (Niagara|SpawnSystematLocation "/WildEstateBosses/Gameplay/RiftEffect/NS_WildEstate_Portal_EntryExit.NS_WildEstate_Portal_EntryExit" (Transformation|GetActorLocation)))))

(event Custom|SetPropVisibiilty
  (params (IsVisible :type Boolean))
  (Variables|Default|SetPropVisibility IsVisible))

(event EventBeginPlay
  (|CallParentFunction)
  (CharacterAnimation|SetWeaponProxyTableOverride (Weapon|GetCurrentWeapon) "/Game/Animation/Game/MainPlayerFN/ProxyTables/ClimbSpill_Pistol_PT.ClimbSpill_Pistol_PT")
  (CallFunction|SpawnRiftEffect)
  (CallFunction|ForceHideHealthBar 0.0))

(event Collision|EventActorBeginOverlap
  (params (OtherActor :type Actor%20Object%20Reference))
  (|CallParentFunction OtherActor) ; This node is disabled and will not be called.\nDrag off pins to build functionality.
)

(event EventTick
  (params (DeltaSeconds :type Float%20%28single-precision%29))
  (|CallParentFunction DeltaSeconds) ; This node is disabled and will not be called.\nDrag off pins to build functionality.
)

(event EventEndPlay
  (params (EndPlayReason :type EEndPlayReason%20Enum :default "Destroyed"))
  (|CallParentFunction EndPlayReason))

(fn OnRep_PropVisibility
  (Rendering|SetVisibility (Variables|Default|GetBalloon) (Variables|Default|GetPropVisibility) true))

(fn ConstructionScript
  (|CallParentFunction))
